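# Physics asset for the Forest Dweller backpack cosmetic, giving its dangling feather and cloth bones physics bodies and joint limits.
# Two boxed feather bodies attach to the pack body, which combines a sphere and a tilted box.
# Three cloth spheres form a fork: cloth_a attaches to the pack with tighter swing, while cloth_b and cloth_c branch from it with looser limits.
# The root carries a broad, thin box and has no constraint.
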PhysicsAsset /BRCosmetics/Accessories/FORT_Backpacks/Backpack_M_MED_Forest_Dweller/Meshes/M_MED_Forest_Dweller_Physics
bodies {
 dyn_cloth_a [{shapeType=Sphere, radius=2, center=<0.5349 0 0>}]
 dyn_cloth_b [{shapeType=Sphere, radius=1.6756, center=<2 0 0>}]
 dyn_cloth_c [{shapeType=Sphere, radius=4.0981, center=<1.4122 0 0>}]
 dyn_feather_b [{shapeType=Box, extentX=0.8631, extentY=5.3338, extentZ=5.0956, center=<2.4933 0 0>, rotation=<rot 90 0 0>}]
 dyn_feathers_a [{shapeType=Box, extentX=0.5851, extentY=5.4814, extentZ=4.6537, center=<2.4582 0 0>, rotation=<rot 90 0 0>}]
 pack [
  {shapeType=Sphere, radius=8.6011, center=<-6.8768 -5.9704 -2.5502>}
  {shapeType=Box, extentX=0.6634, extentY=8.4047, extentZ=12.9999, center=<0 -6.9032 6.5455>, rotation=<rot 39.2615 0 0>}
 ]
 root [{shapeType=Box, extentX=43.3663, extentY=1.8476, extentZ=43.3663, center=<0 0 -3.5216>, rotation=<rot -0.1608 1.4872 6.1697>}]
}
constraints [
 {bone1Name=dyn_feathers_a, bone2Name=pack, swing1LimitDegrees=45, swing2LimitDegrees=45, twistLimitDegrees=5}
 {bone1Name=dyn_feather_b, bone2Name=pack, swing1LimitDegrees=45, swing2LimitDegrees=45, twistLimitDegrees=5}
 {bone1Name=dyn_cloth_a, bone2Name=pack, swing1LimitDegrees=19, swing2LimitDegrees=10, twistLimitDegrees=5}
 {bone1Name=dyn_cloth_b, bone2Name=dyn_cloth_a, swing1LimitDegrees=45, swing2LimitDegrees=70, twistLimitDegrees=10}
 {bone1Name=dyn_cloth_c, bone2Name=dyn_cloth_a, swing1LimitDegrees=45, swing2LimitDegrees=70, twistLimitDegrees=10}
]
