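# Physics asset for the Forest Dweller backpack cosmetic mesh, giving it a simple collision shape.
# A single body on the pack bone holds one box that is very thin along Y and slightly offset on that axis, forming a flat slab.
# It has no constraints, so no joints link it to other bodies.
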PhysicsAsset /BRCosmetics/Accessories/FORT_Backpacks/Backpack_M_MED_Forest_Dweller/Meshes/M_MED_Forest_Dweller_Physics1
bodies {pack [{shapeType=Box, extentX=35.6871, extentY=1.0042, extentZ=51.8389, center=<0 2.599 0>}]}
constraints []
